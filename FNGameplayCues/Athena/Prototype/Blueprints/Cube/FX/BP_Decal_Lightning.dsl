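; BP_Decal_Lightning is a prototype Cube FX Blueprint actor, named as a lightning decal.
; On BeginPlay it adds a random local rotation of up to a full turn, so spawned instances differ in orientation.
; Tick and the construction script contain no logic.
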(event EventBeginPlay
  (Transformation|AddActorLocalRotation 0 (Math|Rotator|MakeRotator (Math|Random|RandomFloatinRange 0.0 3.6e+02)) false false))

(event EventTick
  (params (DeltaSeconds :type Float%20%28single-precision%29)))

(fn ConstructionScript)
